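# Material function that outputs a white-to-black distance fade from the camera for use in Fortnite light functions.
# It divides the distance between camera position and world position by a constant falloff range, then clamps and inverts the result.
# Pixels near the camera are white, fading linearly to black at the range limit.
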MaterialFunction /FigmentEnvironment/LM_8_1_1/Environments/World/LightFunctions/ForniteLightFunctionFalloff
graph {
 properties {description=White to black fade that the light function uses, libraryCategoriesText [Misc,Fortnite]}
 nodes [
  n1 = Divide(A=Distance(A=CameraPositionWS(), B=WorldPosition().XYZ), B=Constant() {r=4096})
  FunctionOutput(None=OneMinus(None=Clamp(None=@n1)))
 ]
}
