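# Material function for tree materials that adjusts Normal and Specular on the back faces of two-sided geometry.
# It takes a material attributes input, multiplies Normal by TwoSidedSign to flip it, and multiplies Specular by the saturated sign to zero it on back faces.
# A "Flip Backfaces" static bool input switches both changes on or off, then the attributes are written back to the output.
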MaterialFunction /FigmentEnvironment/LM_8_1_1/Packages/Fortnite_UniqueMaterials_1/MaterialFunctions/Trees/MF_FlipTreeBackfaces
graph {nodes [
  n1 = GetMaterialAttributes(None=FunctionInput() {inputName=Attrbutes, inputType=FunctionInput_MaterialAttributes}) {attributeGetTypes [#1, #2]}
  n2 = TwoSidedSign()
  n3 = Multiply(A=@n1.Specular, B=Saturate(None=@n2))
  n4 = FunctionInput(Preview=StaticBool()) {bUsePreviewValueAsDefault=true, inputName=Flip Backfaces, inputType=FunctionInput_StaticBool, sortPriority=1}
  n5 = Multiply(A=@n1.Normal, B=@n2)
  FunctionOutput(None=SetMaterialAttributes!(
    MaterialAttributes=@n1.MaterialAttributes
    Specular=StaticSwitch(True=@n3, False=@n1.Specular, Value=@n4)
    Normal=StaticSwitch(True=@n5, False=@n1.Normal, Value=@n4)
   ) {attributeSetTypes [#1, #2]})
 ]}
