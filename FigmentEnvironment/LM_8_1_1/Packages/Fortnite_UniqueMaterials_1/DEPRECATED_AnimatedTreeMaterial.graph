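# Deprecated masked, two-sided subsurface master material for Fortnite destructible trees, with leaf animation, optional snow and fuzzy shading.
# A damage percentage thresholds the mask's blue channel to swap bark, optionally reveal damage textures, and blend heightmap-derived normals.
# World position offset combines Pivot Painter leaf rotation, point-masked hit bounce, shrink and shared directional wind.
# A particle switch disables that offset and substitutes smoothstep-eroded opacity driven by particle alpha.
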Material /FigmentEnvironment/LM_8_1_1/Packages/Fortnite_UniqueMaterials_1/DEPRECATED_AnimatedTreeMaterial
graph {
 properties {
  bUseMaterialAttributes=true
  bUsedWithInstancedStaticMeshes=true
  bUsedWithMeshDeformer=true
  bUsedWithMeshParticles=true
  bUsedWithSkeletalMesh=true
  blendMode=BLEND_Masked
  opacityMaskClipValue=0.5
  pixelDepthOffsetMode=PDOM_Legacy
  shadingModel=MSM_Subsurface
  twoSided=true
  usedShadingModels=Subsurface
 }
 nodes [
  n1 = TextureSampleParameter2D() {group=Textures, parameterName=Diffuse, texture=/FigmentEnvironment/LM_8_1_1/Environments/Props/Trees/Textures/Tree_02_D}
  n2 = TextureObjectParameter() {group=Damage, parameterName=Trunk Damage Texture, texture=/FigmentEnvironment/LM_8_1_1/Packages/User_Built_Props/Textures/wood6}
  n3 = VectorParameter() {group=Damage, parameterName=Trunk damage texture tiling, defaultValue=<rgba 0 0 0.3 0.3>}
  n4 = MaterialFunctionCall!(
   UVs
   TextureIn=@n2
   UpperLeftCorner=ComponentMask(None=@n3.RGB) {g=true, r=true}
   LowerRightCorner=AppendVector(A=@n3.B, B=@n3.B)
   ExportFloat4
  ) {materialFunction=/Engine/Functions/Engine_MaterialFunctions01/Texturing/TextureCropping}
  n5 = VectorParameter() {group=Damage, parameterName=BarkColorDamage, defaultValue=<rgba 1 0.706 0.3688 1>}
  n6 = ScalarParameter() {group=Code, parameterName=Code_CustomDamagePercentage}
  n7 = OneMinus(None=@n6)
  n8 = TextureSampleParameter2D() {
   group=Textures
   parameterName=MaskTexture
   samplerType=SAMPLERTYPE_Masks
   texture=/FigmentEnvironment/LM_8_1_1/Environments/Props/Trees/Textures/Tree_02_M
  }
  n9 = Constant() {r=1}
  n10 = If(A=@n7, B=@n8.B, "A > B"=@n9, "A == B"=@n9, "A < B"=Constant()) {desc=base interpolation, equalsThreshold=0.001}
  n11 = OneMinus(None=@n10)
  n12 = SphereMask(A=@n7, B=@n8.B, Radius=Constant() {r=0.05}) {hardnessPercent=0}
  n13 = StaticSwitchParameter(True=Clamp(None=Subtract(A=@n11, B=@n12)), False=Constant()) {defaultValue=true, group=Damage, parameterName=DestroyBark}
  n14 = LinearInterpolate(A=LinearInterpolate(A=@n1.RGB, B=Multiply(A=@n4.Cropped, B=@n5.RGB) {desc=colorize}, Alpha=@n13) {desc=lerp in bark}, B=@n1.RGB, Alpha=@n8.G) {desc=leaves}
  n15 = TextureSampleParameter2D() {group=Textures, parameterName=Damage reveal texture, texture=/FigmentEnvironment/LM_8_1_1/Environments/Props/Trees/Textures/PineBranches_D}
  n16 = OneMinus(None=@n10)
  n17 = LinearInterpolate(A=@n14, B=@n15.RGB, Alpha=@n16) {desc=lerp damage reveal texture}
  n18 = TextureSampleParameter2D() {
   desc="spec, metallic,roughness\n"
   group=Specular
   parameterName=SpecularMasks
   samplerType=SAMPLERTYPE_Masks
   texture=/FigmentEnvironment/LM_8_1_1/Packages/Fortnite_UniqueMaterials_1/Textures/T_Fortnite_Default_S
  }
  n19 = TextureSampleParameter2D() {
   desc="spec, metallic,roughness\n"
   group=Specular
   parameterName=Damage_SpecularMasks
   samplerType=SAMPLERTYPE_Masks
   texture=&T_Fortnite_Default_S
  }
  n20 = TextureSampleParameter2D() {
   group=Textures
   parameterName=NormalMap
   samplerType=SAMPLERTYPE_Normal
   texture=/FigmentEnvironment/LM_8_1_1/Environments/Props/Shrubs/Textures/Shrub_05_N
  }
  n21 = StaticSwitchParameter(True=@n20.RGB, False=Constant3Vector() {constant=<rgba 0 0 1 0>}) {group=Textures, parameterName=UseNormalMap}
  n22 = TextureSampleParameter2D() {
   group=Textures
   parameterName=Damage reveal normal texture
   samplerType=SAMPLERTYPE_Normal
   texture=/FigmentEnvironment/LM_8_1_1/Environments/Props/Trees/Textures/PineBranches_N
  }
  n23 = StaticSwitchParameter(True=LinearInterpolate(A=@n21, B=@n22.RGB, Alpha=@n16), False=@n21) {group=Damage, parameterName=Use Damage Reveal Texture}
  n24 = TextureObjectParameter() {parameterName=MaskTexture, samplerType=SAMPLERTYPE_Masks, texture=&Tree_02_M}
  n25 = MaterialFunctionCall!(
   Height Map=@n24
   Normal Map Intensity=Constant() {r=-8}
   Height Map UV Offset=ScalarParameter() {defaultValue=0.001, group=Damage, parameterName=DamageNormalGenerationOffset}
   Coordinates
   Height Map Channel Selector=Constant4Vector() {constant=<rgba 0 0 1 0>}.RGBA
  ) {materialFunction=/Engine/Functions/Engine_MaterialFunctions03/Procedurals/NormalFromHeightmap}
  n26 = Multiply(A=SphereMask(A=@n7, B=@n8.B) {attenuationRadius=0.2, desc=normal mask, hardnessPercent=0}, B=Clamp(None=Multiply(A=@n6) {constB=20}))
  n27 = MaterialFunctionCall!(
   -----------Base Shader-----------
   BaseColor1=StaticSwitchParameter(True=StaticSwitchParameter(True=@n17, False=@n14) {group=Damage, parameterName=Use Damage Reveal Texture}, False=@n1.RGB) {defaultValue=true, group=Damage, parameterName=Use Damage Effects}
   Base Color Alpha
   SpecularMasks1=StaticSwitchParameter(True=LinearInterpolate(A=@n18.RGB, B=@n19.RGB, Alpha=@n13) {desc=damaged bark roughness}, False=@n18.RGB) {defaultValue=true, group=Damage, parameterName=Use Damage Effects}
   Emissive
   Normals1=StaticSwitchParameter(True=LinearInterpolate(A=@n23, B=@n25.Result, Alpha=@n26), False=@n23) {defaultValue=true, group=Damage, parameterName=Use Damage Effects}
   World Position Offset
   DiffuseFresnelAmount=Constant() {r=0.125}
   -----------Destruction------------
   ------Multiple Diffuse Textures-----
   BaseColor2=Constant() {r=0.3}
   BaseColor3
   BaseColor4
   BaseColor5
   BaseColor6
   BaseColor7
   BaseColor8
   " - - - - - - - -"
   SpecularMasks2
   SpecularMasks3
   SpecularMasks4
   SpecularMasks5
   SpecularMasks6
   SpecularMasks7
   SpecularMasks8
   " - - - - - - - - "
   normal2
   Normal3
   Normal4
   Normal5
   Normal6
   Normal7
   Normal8
   " - - - - - - - - 1"
   Use 2 Textures=StaticBoolParameter() {group=Textures, parameterName=2 Textures}
   Use 3 Textures
   Use 4 Textures
   Use 5 Textures
   Use 6 Textures
   Use 7 Textures
   Use 8 Textures
   Use 9 Textures
   Use 10 Textures
   ExposeLerpFunctionForInvertingTheHitDistanceCalc
  ) {materialFunction=/FigmentEnvironment/LM_8_1_1/Packages/Fortress_SharedMaterials/Base_Material_Functions/DEPRECATED_BaseShaderFunction}
  n28 = LinearInterpolate(
   A=ScalarParameter() {group=Snow, parameterName=SnowNormalContrast1}
   B=ScalarParameter() {defaultValue=1, group=Snow, parameterName=SnowNormalContrast2}
   Alpha=ComponentMask(None=Transform(None=@n27.Tangent Normal)) {b=true}
  )
  n29 = Clamp(None=DotProduct(A=CameraVectorWS(), B=PixelNormalWS()))
  n30 = Power(Base=OneMinus(None=@n29)) {constExponent=6}
  n31 = ScalarParameter() {defaultValue=0.8, group=Fuzziness, parameterName=Fuzzy Edge Brightness}
  n32 = Multiply(A=@n29, B=ScalarParameter() {defaultValue=0.8, parameterName=Fuzzy Core Darkness}) {constB=0.9, desc=core}
  n33 = Multiply(A=Add(A=Multiply(A=@n30, B=@n31) {constB=0.8, desc=edge}, B=OneMinus(None=@n32)), B=@n27.Diffuse) {desc=diffuse}
  n34 = Multiply(
   A=ScalarParameter() {defaultValue=1, parameterName=Diffusebrightness}
   B=QualitySwitch(Default=StaticSwitchParameter(True=@n33, False=@n27.Diffuse) {group=Fuzziness, parameterName=Use Fuzzy Shading}, Low=@n27.Diffuse)
  )
  n35 = LinearInterpolate(
   A=ScalarParameter() {defaultValue=0.3802, group=Snow, parameterName=SnowDiffuseContrast1}
   B=ScalarParameter() {defaultValue=8.5895, group=Snow, parameterName=SnowDiffuseContrast2}
   Alpha=ComponentMask(None=@n34) {r=true}
  )
  n36 = LinearInterpolate(
   A=@n34
   B=VectorParameter() {group=Snow, parameterName=SnowColor, defaultValue=<rgba 0.7262 0.7933 0.83 1>}.RGB
   Alpha=Multiply(A=Clamp(None=@n28), B=Clamp(None=@n35))
  )
  n37 = StaticBool() {value=true}
  n38 = Transform(None=@n27.Tangent Normal)
  n39 = ScalarParameter() {defaultValue=0.1, group=Specular, parameterName=SpecRoughnessMin}
  n40 = ScalarParameter() {defaultValue=1, group=Specular, parameterName=SpecRoughnessMax}
  n41 = TransformPosition(None=WorldPosition() {worldPositionShaderOffset=WPT_ExcludeAllShaderOffsets}.XYZ) {transformSourceType=TRANSFORMPOSSOURCE_World}
  n42 = VectorParameter() {group=WorldPositionOffset, parameterName=Local position of bounce mask, defaultValue=<rgba 0 0 0 1>}
  n43 = SphereMask(
   A=Divide(
    A=Subtract(A=@n41, B=@n42.RGB)
    B=VectorParameter() {group=WorldPositionOffset, parameterName=Bounce Mask Radius XYZ, defaultValue=<rgba 128 128 128 1>}.RGB
   )
   B=Constant()
   Radius=Constant() {r=1}
   Hardness=ScalarParameter() {defaultValue=0.5, group=WorldPositionOffset, parameterName=Bounce Mask Hardness}
  ) {hardnessPercent=50}
  n44 = Add(A=@n43, B=@n27.emissive)
  n45 = ScalarParameter() {defaultValue=0.25, group=SSS, parameterName=Amount Of SubSurface Scattering}
  n46 = OneMinus(None=@n45)
  n47 = StaticComponentMaskParameter(None=AppendVector(A=@n8.RGB, B=@n8.RGB)) {defaultG=true, desc=black bark and white leaves, group=Textures, parameterName=BarkMaskTextureChannel}
  n48 = Multiply(A=@n10, B=@n47)
  n49 = Multiply(
   A=@n45
   B=StaticSwitchParameter(True=StaticSwitchParameter(True=@n48, False=@n47) {group=Damage, parameterName=Use Damage Reveal Texture}, False=Constant() {r=1}) {defaultValue=true, group=Damage, parameterName=Use Damage Effects}
  )
  n50 = VectorParameter() {group=SSS, parameterName=SSS Color, defaultValue=<rgba 0.489 1 0.05 1>}
  n51 = Multiply(A=@n49, B=Multiply(A=ComponentMask(None=@n27.Diffuse) {r=true}, B=@n50.RGB))
  n52 = Multiply(A=@n49, B=@n50.RGB)
  n53 = MaterialFunctionCall!(WindVector, Max Dist for Parent Piv) {materialFunction=/Engine/Functions/Engine_MaterialFunctions02/PivotPainter/PivotPainter_TreeData}
  n54 = MaterialFunctionCall() {materialFunction=&PivotPainter_TreeData}
  n55 = ScalarParameter() {defaultValue=0.5, group=LeafAnim, parameterName=LeafGrowthAnimationSpeed}
  n56 = Time()
  n57 = Distance(A=@n54.Leaf Piv Position, B=WorldPosition().XYZ)
  n58 = ScalarParameter() {defaultValue=-0.0001, group=LeafAnim, parameterName=LeafAnimationOffsetFromPivot}
  n59 = ScalarParameter() {defaultValue=0.01, group=LeafAnim, parameterName=LeafFlowGrad}
  n60 = Multiply(A=Time(), B=ScalarParameter() {defaultValue=0.5, group=LeafAnim, parameterName=LeafAnimationSpeed})
  n61 = Add(
   A=StaticSwitchParameter(True=Multiply(A=@n55, B=@n56), False=ScalarParameter() {group=LeafAnim, parameterName=Leaf Animation Time}) {group=Debug, parameterName=DebugGrowth}
   B=Add(A=Multiply(A=@n57, B=@n58) {constB=-0.0001}, B=Add(A=@n59, B=@n60))
  )
  n62 = Add(A=@n61, B=ComponentMask(None=TextureCoordinate() {coordinateIndex=3}) {desc=Leaf Z height The rest is black, r=true})
  n63 = Add(A=Sine(None=@n61) {period=0.25}, B=Sine(None=Add(A=@n61, B=@n62) {constB=0.25}) {period=2.3333})
  n64 = ScalarParameter() {defaultValue=1, group=LeafAnim, parameterName=LeafAnimationAmount}
  n65 = Multiply(
   A=RotateAboutAxis(
    NormalizedRotationAxis=@n54.Leaf X-Axis Vector
    RotationAngle=Multiply(A=@n63, B=@n64)
    PivotPoint=@n54.Leaf X-Axis Vector
    Position=WorldPosition().XYZ
   ) {period=28}
   B=@n53.Leaf Mask
  ) {desc=MaskTrunk}
  n66 = OneMinus(None=@n43)
  n67 = MaterialFunctionCall!(
   CrackDamageShader
   Code_HitAnimation
   Code_DestructionAmount
   Code_UseUvForDamage
   Code_DeformationVector
   Code_UseHitEffects
   WSDeformationCenter
   WSDamageRadius
   Diffuse
   Code_Emissive
   ScalarModDamageCracks
   ScalarOffsetforDamageCracks
   ExposeLerpFunctionForInvertingTheHitDistanceCalc
   InvertDamageFalloff
   WorldPosition
   DistanceCullFadeAnimationOffset
  ) {materialFunction=/FigmentEnvironment/LM_8_1_1/Packages/jl_tests/Fortnite_HitEffects}
  n68 = Add(
   A=StaticSwitchParameter(True=Multiply(A=@n53.Black Mask, B=@n65), False=Constant()) {defaultValue=true, group=LeafAnim, parameterName=Animate}
   B=StaticSwitchParameter(True=Multiply(A=@n66, B=@n67.Hit WPO), False=@n67.Hit WPO) {defaultValue=true, group=WorldPositionOffset, parameterName=Mask Bounce Around a Point}
  )
  n69 = OneMinus(None=Power(Base=ScalarParameter() {group=WorldPositionOffset, parameterName=DistanceCullFadeAnimationOffset}) {constExponent=5})
  n70 = Constant() {r=1}
  n71 = StaticBool()
  n72 = StaticBool() {value=true}
  n73 = Add(
   A=Add(A=Multiply(A=@n68, B=@n69) {desc=remove the animation from the tree as it shrinks}, B=@n67.Shrink WPO) {desc=add the shrinking effect}
   B=Multiply(
    A=StaticSwitchParameter(True=@n66, False=@n70) {defaultValue=true, group=WorldPositionOffset, parameterName=Mask Bounce Around a Point}
    B=MaterialFunctionCall!(
     Force Vector
     Grass/Flowers?=@n71
     Directional Wind Falloff From Leaf Pivot Position Radius
     Additional Motion Mask
     grass Pivot
     Directional Offset Amount
     Use Root Rotation=@n72
     Wind Vector
     Element Vector
     Length Wise Animation Offset
     Length Wise Animation Offset1
     Grass Pivot Animation Offset1
    ) {materialFunction=/FigmentEnvironment/LM_8_1_1/Packages/Fortnite_UniqueMaterials_1/MaterialFunctions/SharedDirectionalWindVS}.WPO
   )
  )
  n74 = Constant()
  n75 = Time()
  n76 = MaterialFunctionCall(Float2=TextureCoordinate()) {materialFunction=&BreakOutFloat2Components}
  n77 = Multiply(A=ScalarParameter() {defaultValue=1, parameterName=Smooth Step U Scale}, B=@n76.R)
  n78 = Add(A=Multiply(A=ScalarParameter() {parameterName=Smooth Step U Speed}, B=@n75), B=@n77)
  n79 = Multiply(A=@n76.G, B=ScalarParameter() {defaultValue=1, parameterName=Smooth Step V Scale})
  n80 = Add(A=@n79, B=Multiply(A=@n75, B=ScalarParameter() {parameterName=Smooth Step V Speed}))
  n81 = TextureSampleParameter2D(UVs=StaticSwitch(True=AppendVector(A=@n78, B=@n80), False=AppendVector(A=@n77, B=@n79), Value=StaticBoolParameter() {parameterName=Smooth Step Pans})) {parameterName=SmoothStep_Texture, texture=/FigmentEnvironment/LM_8_1_1/VisualThreatManager/StormVisuals/Textures/2ndInfectedMapLinear}
  n82 = OneMinus(None=Multiply(A=ScalarParameter() {defaultValue=1, parameterName=Smooth Step Constant}, B=ParticleColor().A))
  n83 = MaterialFunctionCall(Alpha=@n81.G, Min=Add(A=@n82) {constB=-0.1}, Max=@n82) {materialFunction=&SmoothStep}
  n84 = TextureSampleParameter2D() {parameterName=Opacity Mask, samplerType=SAMPLERTYPE_Masks, texture=/FigmentEnvironment/LM_8_1_1/Environments/Props/Trees/Textures/Tree_21_M}
  n85 = Multiply(A=@n83.Result, B=StaticComponentMaskParameter(None=@n84.RGB) {defaultR=true, parameterName=Opacity Mask Channel})
  n86 = StaticBoolParameter() {parameterName=Use Opacity Mask}
  n87 = TextureSampleParameter2D() {group=Textures, parameterName=Damage reveal texture, texture=&PineBranches_D}
  n88 = OneMinus(None=@n8.G)
  n89 = Multiply(A=Clamp(None=Add(A=@n10, B=@n88)), B=@n8.R)
  n90 = StaticSwitchParameter(True=Add(A=@n87.A, B=@n89), False=@n89) {group=Damage, parameterName=Use Damage Reveal Texture}
  n91 = TextureSampleParameter2D() {
   group=Textures
   parameterName=MaskTexture
   samplerType=SAMPLERTYPE_Masks
   texture=&Tree_02_M
  }
  n92 = MakeMaterialAttributes!(
   BaseColor=LinearInterpolate(
    A=StaticSwitchParameter(True=@n36, False=@n34) {parameterName=UseSnow}
    Alpha=MaterialFunctionCall!(
     Use_ExtraSnowMask
     UV_WorldMask
     SnowStartBiasOverride
     UseHeightLerp
     LandscapeBlend_Height
     Landscape_RockMask
     Snow_Mask
     Simple_UseWorldZNormal=@n37
     Simple_WorldNormal=@n38
     WorldPosition
    ) {materialFunction=/FigmentEnvironment/LM_8_1_1/Athena/Environments/Landscape/MaterialFunctions/Snow/MF_Athena_Holiday_SnowFall}.Simple_SnowMask
   ) {constB=0.8}
   Metallic=@n27.Metalness
   Specular=@n27.Metalness
   Roughness=QualitySwitch(Default=LinearInterpolate(A=@n39, B=@n40, Alpha=@n27.Roughness), Low=Constant() {r=1})
   Anisotropy
   EmissiveColor=StaticSwitchParameter(
    True=StaticSwitchParameter(True=@n44, False=@n27.emissive) {group=WorldPositionOffset, parameterName=Debug Mask Location}
    False=@n27.emissive
   ) {defaultValue=true, group=WorldPositionOffset, parameterName=Mask Bounce Around a Point}
   Opacity=StaticSwitchParameter(True=Multiply(A=@n27.Diffuse, B=@n46), False=@n46) {defaultValue=true, group=SSS, parameterName=Reference Diffuse In SSS}
   OpacityMask=StaticSwitchParameter(
    True=StaticSwitch(True=@n85, False=@n83.Result, Value=@n86)
    False=StaticSwitchParameter(True=@n90, False=@n91.R) {defaultValue=true, group=Damage, parameterName=Use Damage Effects}
   ) {group=ParticleShader, parameterName=IsParticleShader?}
   Normal=@n27.Tangent Normal
   Tangent
   WorldPositionOffset=StaticSwitchParameter(True=Constant(), False=QualitySwitch(Default=@n73, Low=@n74, High=@n73, Medium=@n73)) {group=ParticleShader, parameterName=IsParticleShader?}
   SubsurfaceColor=Multiply(
    A=StaticSwitchParameter(True=@n51, False=@n52) {defaultValue=true, group=SSS, parameterName=Reference Diffuse In SSS}
    B=MaterialFunctionCall() {materialFunction=/FigmentEnvironment/LM_8_1_1/Environments/World/LightFunctions/ForniteLightFunctionFalloff}.Result
   )
   ClearCoat
   ClearCoatRoughness
   AmbientOcclusion
   Refraction
   CustomizedUVs_0
   CustomizedUVs_1
   CustomizedUVs_2
   CustomizedUVs_3
   CustomizedUVs_4
   CustomizedUVs_5
   CustomizedUVs_6
   CustomizedUVs_7
   PixelDepthOffset
   ShadingModel
   Displacement
  )
  n93 = GetMaterialAttributes(None=@n92) {attributeGetTypes [#1]}
  n94 = SetMaterialAttributes!(
   MaterialAttributes=@n93.MaterialAttributes
   Base Color=Multiply(A=@n93.Base Color, B=VectorParameter() {parameterName=BaseColor, defaultValue=<rgba 1 1 1 1>}.RGB)
  ) {attributeSetTypes [#1]}
  MaterialFunctionCall!(MA In=MaterialFunctionCall!(
    Attrbutes=StaticSwitchParameter(True=@n94, False=@n92) {parameterName=ModulateBaseColor}
    Flip Backfaces=StaticBoolParameter() {parameterName=Flip Backfacing Normals}
   ) {materialFunction=/FigmentEnvironment/LM_8_1_1/Packages/Fortnite_UniqueMaterials_1/MaterialFunctions/Trees/MF_FlipTreeBackfaces}.Result) {materialFunction=/FigmentEnvironment/LM_8_1_1/Weapons/Prototype/ObjectMover/Materials/MF_CreativeMoveTool}
  Multiply(A=@n54.Leaf flow Grad, B=@n59)
  ScalarParameter() {group=Code, parameterName=Code_DestructionAmount}
  ScalarParameter() {group=Code, parameterName=WSDamageRadius}
  ScalarParameter() {group=Code, parameterName=Code_HitAnimation}
  VectorParameter() {group=Code, parameterName=Code_DeformationCenter}
  VectorParameter() {group=Code, parameterName=Code_DeformationVector}
 ]
}
